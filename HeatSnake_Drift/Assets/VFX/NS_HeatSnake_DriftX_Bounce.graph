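# Niagara system for the HeatSnake DriftX bounce effect, apparently an impact burst judging by its name and emitters.
# Four emitters are enabled: LW_Spike, LW_Smoke_Ground, chunks and Mesh_Sparks001, named for spikes, ground smoke, chunks and mesh sparks.
# Light, debris, snowflake and "s" emitters remain but are disabled.
# SystemState is the only listed module, with no user variables and no renderer details.
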NiagaraSystem /HeatSnake_Drift/Assets/VFX/NS_HeatSnake_DriftX_Bounce
dependencies {usedDataInterfaces [], usedDynamicInputs [], usedModules [&SystemState]}
emitters {
 LW_Debris {bEnabled=false, rendererClasses [], renderers []}
 LW_Smoke_Ground {rendererClasses [], renderers []}
 LW_SnowFlakes {bEnabled=false, rendererClasses [], renderers []}
 LW_Spike {rendererClasses [], renderers []}
 Light001 {bEnabled=false, rendererClasses [], renderers []}
 Mesh_Sparks001 {rendererClasses [], renderers []}
 chunks {rendererClasses [], renderers []}
 s {bEnabled=false, rendererClasses [], renderers []}
}
summary {
 systemName=NS_HeatSnake_DriftX_Bounce
 emitters [
  {emitterName=LW_Spike, rendererClasses []}
  {emitterName=LW_Smoke_Ground, rendererClasses []}
  {bEnabled=false, emitterName=Light001, rendererClasses []}
  {emitterName=chunks, rendererClasses []}
  {bEnabled=false, emitterName=LW_Debris, rendererClasses []}
  {bEnabled=false, emitterName=LW_SnowFlakes, rendererClasses []}
  {bEnabled=false, emitterName=s, rendererClasses []}
  {emitterName=Mesh_Sparks001, rendererClasses []}
 ]
 userVariables []
}
