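# Niagara system for the HeatSnake DriftX explosion effect, layering many emitters into one burst.
# Active emitters are named for a main explosion, fire, ground smoke and shockwave, spikes, debris, chunks, mesh sparks, a light and a decal.
# The snowflake emitter and one named "s" are disabled.
# No user variables are exposed, and System State is the only module recorded, with no renderer details.
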NiagaraSystem /HeatSnake_Drift/Assets/VFX/NS_HeatSnake_DriftX_Explosions
dependencies {usedDataInterfaces [], usedDynamicInputs [], usedModules [&SystemState]}
emitters {
 Decal {rendererClasses [], renderers []}
 LW_Debris {rendererClasses [], renderers []}
 LW_Explosion_Main {rendererClasses [], renderers []}
 LW_Fire {rendererClasses [], renderers []}
 LW_Ground_Shockwave {rendererClasses [], renderers []}
 LW_Smoke_Ground {rendererClasses [], renderers []}
 LW_SnowFlakes {bEnabled=false, rendererClasses [], renderers []}
 LW_Spike {rendererClasses [], renderers []}
 Light001 {rendererClasses [], renderers []}
 Mesh_Sparks001 {rendererClasses [], renderers []}
 chunks {rendererClasses [], renderers []}
 s {bEnabled=false, rendererClasses [], renderers []}
}
summary {
 systemName=NS_HeatSnake_DriftX_Explosions
 emitters [
  {emitterName=LW_Spike, rendererClasses []}
  {emitterName=LW_Smoke_Ground, rendererClasses []}
  {emitterName=Light001, rendererClasses []}
  {emitterName=chunks, rendererClasses []}
  {emitterName=LW_Debris, rendererClasses []}
  {bEnabled=false, emitterName=LW_SnowFlakes, rendererClasses []}
  {bEnabled=false, emitterName=s, rendererClasses []}
  {emitterName=Decal, rendererClasses []}
  {emitterName=Mesh_Sparks001, rendererClasses []}
  {emitterName=LW_Explosion_Main, rendererClasses []}
  {emitterName=LW_Ground_Shockwave, rendererClasses []}
  {emitterName=LW_Fire, rendererClasses []}
 ]
 userVariables []
}
