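# Niagara system for the HeatSnake DriftX pawn impact effect, an explosion-style burst with cyan default colors.
# Active emitters cover the main explosion, spikes, debris, snowflakes, mesh sparks, a light and one named "s".
# Ground smoke, chunk and decal emitters are disabled.
# User parameters expose global and additional fire colors, scale, fire intensity, heat dissipation and an iconic-blast toggle.
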NiagaraSystem /HeatSnake_Drift/Assets/VFX/NS_HeatSnake_DriftX_Pawn_Impact
dependencies {usedDataInterfaces [], usedDynamicInputs [], usedModules [&SystemState]}
emitters {
 Decal {bEnabled=false, rendererClasses [], renderers []}
 LW_Debris {rendererClasses [], renderers []}
 LW_Explosion_Main {rendererClasses [], renderers []}
 LW_Smoke_Ground {bEnabled=false, rendererClasses [], renderers []}
 LW_SnowFlakes {rendererClasses [], renderers []}
 LW_Spike {rendererClasses [], renderers []}
 Light001 {rendererClasses [], renderers []}
 Mesh_Sparks001 {rendererClasses [], renderers []}
 chunks {bEnabled=false, rendererClasses [], renderers []}
 s {rendererClasses [], renderers []}
}
summary {
 systemName=NS_HeatSnake_DriftX_Pawn_Impact
 emitters [
  {emitterName=LW_Spike, rendererClasses []}
  {bEnabled=false, emitterName=LW_Smoke_Ground, rendererClasses []}
  {emitterName=Light001, rendererClasses []}
  {bEnabled=false, emitterName=chunks, rendererClasses []}
  {emitterName=LW_Debris, rendererClasses []}
  {emitterName=LW_SnowFlakes, rendererClasses []}
  {emitterName=s, rendererClasses []}
  {bEnabled=false, emitterName=Decal, rendererClasses []}
  {emitterName=Mesh_Sparks001, rendererClasses []}
  {emitterName=LW_Explosion_Main, rendererClasses []}
 ]
 userVariables [
  User.Additional Fire Color {type=LinearColor, description="", defaultValue {struct=LinearColor, value=<rgba 0 1.8109 4 1>}}
  User.Fire Intensity {type=NiagaraFloat, defaultValue=1, description=""}
  User.Global Color {type=LinearColor, description="", defaultValue {struct=LinearColor, value=<rgba 0 1.7202 4 1>}}
  User.Global Scale {type=NiagaraFloat, defaultValue=1, description=""}
  User.Heat Dissipation {type=NiagaraFloat, defaultValue=0.5, description=""}
  User.Use Iconic Blast {type=NiagaraBool, defaultValue=true}
 ]
}
