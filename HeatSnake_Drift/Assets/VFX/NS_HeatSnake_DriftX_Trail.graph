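# Niagara trail effect for HeatSnake DriftX: a short-lived, HDR orange ribbon that takes a skeletal mesh user parameter.
# One emitter spawns through an anim-trail function and distance-based SpawnPerUnit, with rate spawning disabled.
# A companion function initializes anim-trail bone indices on each new particle.
# A curve over normalized execution index tapers ribbon width, and a sprite renderer sits alongside the ribbon renderer.
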NiagaraSystem /HeatSnake_Drift/Assets/VFX/NS_HeatSnake_DriftX_Trail
dependencies {usedDataInterfaces [&NiagaraDataInterfaceCurve], usedDynamicInputs [], usedModules [&SystemState]}
emitters {Minimal {
  emitterUpdateScript [
   &EmitterState()
   &FN_SpawnAnimTrail()
   &SpawnRate(SpawnRate=6, Use Spawn Probability=false) {enabled=false}
   &SpawnPerUnit(Module.Use Max Movement Threshold=true, Module.Use Movement Tolerance=false, Spawn Spacing=15, Use Spawn Probability=false)
  ]
  particleSpawnScript [
   &InitializeParticle(
    Color=<rgba 2 0.5849 0 1>
    Lifetime=0.3
    Ribbon Width=8
    Ribbon Width Mode=::NewEnumerator1
    UsePositionOffset=false
    Write Material Random=true
   )
   &FN_InitializeAnimTrailBoneIndecies()
  ]
  particleUpdateScript [
   &ParticleState()
   &ScaleRibbonWidth(Ribbon Width Scale=&FloatFromCurve(
     FloatCurve=DataInterface() {propertyValues {Curve {keys [<at 0 1>,<at 0.9976 0.2118>]}}}
     CurveIndex=&ReturnNormalizedExecIndex()
    ))
  ]
  rendererClasses [&NiagaraRibbonRendererProperties, &NiagaraSpriteRendererProperties]
  renderers [
   {rendererClass=&NiagaraRibbonRendererProperties}
   {rendererClass=&NiagaraSpriteRendererProperties, rendererIndex=1}
  ]
 }}
summary {
 systemName=NS_HeatSnake_DriftX_Trail
 emitters [{emitterName=Minimal, rendererClasses [&NiagaraRibbonRendererProperties, &NiagaraSpriteRendererProperties]}]
 userVariables [User.SkeletalMesh {
   type=NiagaraDataInterfaceSkeletalMesh
   defaultValue {
    struct=NiagaraExt_VariableValue_DataInterface
    value {dataInterface=&self:NiagaraDataInterfaceSkeletalMesh_0}
   }
  }]
}
