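# Niagara system for the HeatSnake Drift muzzle flash effect.
# It comprises three Center emitters, Mesh_Sparks and Smoke_Ring, with no renderer details listed.
# User parameters expose a color array, an integer Variant and a float IS_Exotic.
# A system-spawn variable assignment and array-select, lerp and switch dynamic inputs suggest parameter-driven variants within one system, though their wiring is not shown.
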NiagaraSystem /HeatSnake_Drift/Assets/VFX/NS_HeatSnake_Drift_MuzzleFlash
dependencies {
 usedDataInterfaces [&NiagaraDataInterfaceArrayInt32, &NiagaraDataInterfaceArrayFloat3]
 usedDynamicInputs [
  &SelectLinearColorFromArray
  &MergeHelper_VectorSwitch
  &Lerp_Vector
  &SelectIntFromArray
  &Multiply_VectorByFloat
  &SelectVectorFromArray
  &LerpLinearColors
  &MergeHelper_FloatSwitch
  &Lerp_Float
 ]
 usedModules [
  &self:SystemSpawnScript.SystemScriptSource.SystemScriptGraph.NiagaraNodeAssignment_0.SetVariables_#1
  &SystemState
 ]
}
emitters {
 Center {rendererClasses [], renderers []}
 Center001 {rendererClasses [], renderers []}
 Center002 {rendererClasses [], renderers []}
 Mesh_Sparks {rendererClasses [], renderers []}
 Smoke_Ring {rendererClasses [], renderers []}
}
summary {
 systemName=NS_HeatSnake_Drift_MuzzleFlash
 emitters [
  {emitterName=Center, rendererClasses []}
  {emitterName=Mesh_Sparks, rendererClasses []}
  {emitterName=Center001, rendererClasses []}
  {emitterName=Smoke_Ring, rendererClasses []}
  {emitterName=Center002, rendererClasses []}
 ]
 userVariables [
  User.Color Array {
   type=NiagaraDataInterfaceArrayColor
   description=""
   defaultValue {
    struct=NiagaraExt_VariableValue_DataInterface
    value {
     dataInterface=&self:NiagaraDataInterfaceArrayColor_0
     dataInterfaceClass=&NiagaraDataInterfaceArrayColor
    }
   }
  }
  User.IS_Exotic {type=NiagaraFloat, defaultValue=0, description=""}
  User.Variant {type=NiagaraInt32, defaultValue=1}
 ]
}
